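# Translucent, unlit, two-sided material for animated waterfall foam in the Apollo environment, usable on Niagara mesh particles.
# Three panned channels of a waterfall tile texture, with distorted UVs and bump offset on two, layer over a color gradient.
# Emissive adds a Fresnel three-color blend and scales by time of day; opacity uses a UV gradient mask, foam and depth fade.
# Two panned noise samples drive vertex-normal displacement with gradient-varied tiling.
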Material /BRCosmetics/Athena/Effects/Materials/Environment/Apollo/M_Apollo_WaterFall_Foam
graph {
 properties {
  bUsedWithNiagaraMeshParticles=true
  blendMode=BLEND_Translucent
  pixelDepthOffsetMode=PDOM_Legacy
  shadingModel=MSM_Unlit
  twoSided=true
  usedShadingModels=Unlit
 }
 nodes [
  n1 = ComponentMask(None=TextureCoordinate()) {r=true}
  n2 = ComponentMask(None=TextureCoordinate()) {r=true}
  n3 = Clamp(None=OneMinus(None=LinearInterpolate(Alpha=@n2) {constB=1.5}))
  n4 = ComponentMask(None=TextureCoordinate()) {r=true}
  n5 = Add(
   A=ScalarParameter() {defaultValue=0.3, group=WaterFall, parameterName=LerpFoamA}
   B=Power(Base=OneMinus(None=@n4), Exp=ScalarParameter() {defaultValue=1, parameterName=FoamExp}) {constExponent=0.5}
  )
  n6 = MaterialFunctionCall(XY=TextureCoordinate() {uTiling=0.4}) {materialFunction=&Swizzle}
  n7 = Time()
  n8 = Panner(
   Coordinate=Multiply(A=TextureCoordinate(), B=ScalarParameter() {defaultValue=0.7372, group=WaterFall_UVDistort, parameterName=UVDistortTilingA})
   Time=@n7
  ) {speedX=1}
  n9 = TextureCoordinate() {uTiling=0.5}
  n10 = ScalarParameter() {defaultValue=0.8, group=WaterFall_UVDistort, parameterName=UVDistortTilingB}
  n11 = TextureSample(UVs=Panner(Coordinate=Multiply(A=@n9, B=@n10), Time=@n7) {speedX=1}) {texture=/Game/Packages/Fortnite_UniqueMaterials_1/Textures/T_WaterFall_Tile}
  n12 = LinearInterpolate(
   A=Multiply(
    A=TextureSample(UVs=@n8) {texture=&T_WaterFall_Tile}.A
    B=ScalarParameter() {defaultValue=0.0163, group=WaterFall_UVDistort, parameterName=UVDistortAmountA}
   )
   B=Multiply(A=@n11.A, B=ScalarParameter() {defaultValue=0.0125, group=WaterFall_UVDistort, parameterName=UVDistortAmountB})
   Alpha=@n3
  )
  n13 = Multiply(A=Add(A=OneMinus(None=@n6.YX), B=@n12), B=ScalarParameter() {defaultValue=1.8399, group=WaterFall, parameterName=FoamTile_GChannel})
  n14 = Multiply(A=@n7, B=ScalarParameter() {defaultValue=0.5973, group=WaterFall, parameterName=Waterfall Speed})
  n15 = Multiply(
   A=LinearInterpolate(
    A=VectorParameter() {group=WaterFall, parameterName=WaterFallColor, defaultValue=<rgba 0.2881 0.61 0.5573 1>}.RGB
    B=VectorParameter() {group=WaterFall, parameterName=WaterFallColor_Bottom, defaultValue=<rgba 0.2881 0.61 0.5573 1>}.RGB
    Alpha=@n5
   )
   B=TextureSample(UVs=Panner(Coordinate=@n13, Time=@n14) {speedY=-1}) {texture=&T_WaterFall_Tile}.G
  )
  n16 = TextureCoordinate() {uTiling=0.7, vTiling=2.8}
  n17 = Add(A=OneMinus(None=MaterialFunctionCall(XY=@n16) {materialFunction=&Swizzle}.YX), B=@n12)
  n18 = Panner(
   Coordinate=Multiply(
    A=BumpOffset(Coordinate=@n17, Height=@n11.A) {heightRatio=0.125}
    B=ScalarParameter() {defaultValue=1.0601, group=WaterFall, parameterName=FoamTile_RChannel}
   )
   Time=Multiply(A=@n7, B=ScalarParameter() {defaultValue=0.5973, group=WaterFall, parameterName=Waterfall Speed})
  ) {speedY=-1}
  n19 = LinearInterpolate(
   A=@n15
   B=Add(A=@n15, B=TextureSample(UVs=@n18) {texture=&T_WaterFall_Tile}.R)
   Alpha=Clamp(None=@n5)
  )
  n20 = MaterialFunctionCall(XY=TextureCoordinate() {uTiling=0.333, vTiling=2}) {materialFunction=&Swizzle}
  n21 = Add(A=@n20.YX, B=@n12)
  n22 = ScalarParameter() {defaultValue=0.8, group=WaterFall, parameterName=OffsetHeight}
  n23 = Panner(
   Coordinate=Multiply(
    A=BumpOffset(Coordinate=@n21, Height=@n22) {heightRatio=0.2}
    B=ScalarParameter() {defaultValue=1, group=WaterFall, parameterName=FoamTile_BChannel}
   )
   Time=Multiply(A=@n7, B=ScalarParameter() {defaultValue=0.5973, group=WaterFall, parameterName=Waterfall Speed})
  ) {speedY=1}
  n24 = ScalarParameter() {defaultValue=0.125, group=WaterFall, parameterName=LerpFoamB}
  n25 = TextureCoordinate()
  n26 = Power(Base=OneMinus(None=ComponentMask(None=@n25) {r=true}), Exp=ScalarParameter() {defaultValue=1, parameterName=FoamExp}) {constExponent=0.5}
  n27 = LinearInterpolate(
   A=@n19
   B=Add(A=@n19, B=TextureSample(UVs=@n23) {texture=&T_WaterFall_Tile}.B)
   Alpha=Clamp(None=Add(A=@n24, B=@n26))
  ) {constAlpha=0.4}
  n28 = Clamp(None=Add(A=@n3, B=@n27))
  n29 = Multiply(A=Multiply(A=Multiply(A=@n1, B=@n1), B=OneMinus(None=@n1)), B=ComponentMask(None=@n28) {g=true})
  n30 = ScalarParameter() {defaultValue=1, parameterName=OpacityExp}
  DepthFade(
   Opacity=Multiply(A=Power(Base=@n29, Exp=@n30) {constExponent=0.5}, B=ScalarParameter() {defaultValue=1, parameterName=Opacity})
   FadeDistance=ScalarParameter() {parameterName=FadeDistance}
  )
  n31 = MaterialFunctionCall(
   A=VectorParameter() {parameterName="A (V3)"}.RGB
   B=VectorParameter() {parameterName="B (V3)"}.RGB
   C=VectorParameter() {parameterName="C (V3)"}.RGB
   Alpha=Fresnel(ExponentIn=ScalarParameter() {parameterName=ExponentIn})
  ) {materialFunction=&3ColorBlend}
  n32 = StaticSwitchParameter(True=@n28, False=@n27) {parameterName=UseGradientOverlay}
  n33 = ScalarParameter() {defaultValue=1, parameterName="Value_Morning (S)"}
  n34 = ScalarParameter() {defaultValue=1, parameterName="Value_Day (S)"}
  n35 = ScalarParameter() {defaultValue=1, parameterName="Value_Evening (S)"}
  n36 = ScalarParameter() {defaultValue=1, parameterName="Value_Night (S)"}
  MaterialFunctionCall(Emissive=Multiply(
    A=Add(A=@n31.Result, B=@n32)
    B=MaterialFunctionCall(Value_Morning=@n33, Value_Day=@n34, Value_Evening=@n35, Value_Night=@n36) {materialFunction=&Scalar_Time_of_Day}.Result
   )) {materialFunction=&MF_FX_EmissiveAdjustments}
  n37 = MaterialFunctionCall(Float2=TextureCoordinate() {vTiling=2}) {materialFunction=&BreakOutFloat2Components}
  n38 = OneMinus(None=MaterialFunctionCall() {materialFunction=&LinearGradient}.VGradient)
  n39 = ScalarParameter() {defaultValue=5, group=WPO, parameterName=Texture Scale Gradient Multiplier}
  n40 = Clamp(None=Power(Base=Multiply(A=@n38, B=@n39) {constB=5}, Exp=ScalarParameter() {defaultValue=2, group=WPO, parameterName=Texture Scale Gradient Power}))
  n41 = LinearInterpolate(
   A=ScalarParameter() {defaultValue=1, group=WPO, parameterName=ScaleU 2}
   B=ScalarParameter() {defaultValue=1, group=WPO, parameterName=ScaleU}
   Alpha=@n40
  )
  n42 = LinearInterpolate(
   A=ScalarParameter() {defaultValue=1, group=WPO, parameterName=ScaleV 2}
   B=ScalarParameter() {defaultValue=1, group=WPO, parameterName=ScaleV}
   Alpha=@n40
  )
  n43 = AppendVector(A=Multiply(A=@n37.R, B=@n41), B=Multiply(A=@n37.G, B=@n42))
  n44 = ComponentMask(None=VectorParameter() {group=WPO, parameterName=WPO_Pan Speed, defaultValue=<rgba 1 0 0 1>}.RGB) {g=true, r=true}
  n45 = MaterialFunctionCall(Float2=TextureCoordinate() {uTiling=2, vTiling=4}) {materialFunction=&BreakOutFloat2Components}
  n46 = AppendVector(A=Multiply(A=@n45.R, B=@n41), B=Multiply(A=@n45.G, B=@n42))
  n47 = Add(
   A=TextureSample(UVs=Panner(Coordinate=@n43, Speed=@n44) {bFractionalPart=true, speedX=1}) {texture=&2ndInfectedMapLinear}.G
   B=TextureSample(UVs=Panner(Coordinate=@n46, Speed=@n44) {bFractionalPart=true, speedX=1}) {texture=&2ndInfectedMapLinear}.G
  )
  Multiply(
   A=Multiply(A=LinearInterpolate(Alpha=@n47) {constA=-1, constB=2}, B=VertexNormalWS())
   B=ScalarParameter() {defaultValue=1, group=WPO, parameterName=Deform}
  )
  OneMinus(None=@n20.YX)
  VectorParameter() {parameterName=Color}
  VertexColor()
 ]
}
